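# Physics asset for the Progressive Jonesy DMG 2 shirt, simulating secondary motion on collars, shirt panels, belt buckle, shoelaces and right-arm bands.
# Capsules on the spine, pelvis, arms and right thigh form the body.
# Tiny _MSTR anchor capsules join it with free rotation, each driven piece swinging within angular limits from its anchor.
# The tie has only an anchor, and left shoelaces anchor to the pelvis, as no left thigh body exists.
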PhysicsAsset /BRCosmetics/Characters/Player/Male/Medium/Bodies/M_MED_Progressive_Jonesy_DMG_2/Meshes/M_MED_Progressive_Jonesy_DMG_2_Shirt_Physics
bodies {
 clavicle_l [{shapeType=Capsule, length=16.25, radius=6, center=<11.8488 0 1.5882>, rotation=<rot -90 -0.0149 0.0149>}]
 clavicle_r [{shapeType=Capsule, length=16.25, radius=6, center=<-11.8488 0 -1.5882>, rotation=<rot -90 0.0197 179.9803>}]
 dyn_beltbuckle_01_l [{shapeType=Capsule, length=5.269, radius=2.3814, center=<2.494 0 0.0002>, rotation=<rot 0.0013 89.9994 89.9999>}]
 dyn_beltbuckle_01_l_MSTR [{shapeType=Capsule, length=0.01, radius=0.01}]
 dyn_collarOuter_01_l [{shapeType=Capsule, length=5.269, radius=2.3814, center=<2.494 0 0.0002>, rotation=<rot 0.0013 89.9994 89.9999>}]
 dyn_collarOuter_01_l_MSTR [{shapeType=Capsule, length=0.01, radius=0.01}]
 dyn_collarOuter_01_r [{shapeType=Capsule, length=5.269, radius=2.3814, center=<2.494 0 0.0002>, rotation=<rot 0.0013 89.9994 89.9999>}]
 dyn_collarOuter_01_r_MSTR [{shapeType=Capsule, length=0.01, radius=0.01}]
 dyn_collarOuter_bk_01_l [{shapeType=Capsule, length=5.269, radius=2.3814, center=<2.494 0 0.0002>, rotation=<rot 0.0013 89.9994 89.9999>}]
 dyn_collarOuter_bk_01_l_MSTR [{shapeType=Capsule, length=0.01, radius=0.01}]
 dyn_collarOuter_bk_01_r [{shapeType=Capsule, length=5.269, radius=2.3814, center=<2.494 0 0.0002>, rotation=<rot 0.0013 89.9994 89.9999>}]
 dyn_collarOuter_bk_01_r_MSTR [{shapeType=Capsule, length=0.01, radius=0.01}]
 dyn_collar_01_l [{shapeType=Capsule, length=4.519, radius=3.1314, center=<2.494 0 0.0002>, rotation=<rot 0.0013 89.9994 89.9999>}]
 dyn_collar_01_l_MSTR [{shapeType=Capsule, length=0.01, radius=0.01}]
 dyn_collar_01_r [{shapeType=Capsule, length=5.269, radius=2.3814, center=<2.494 0 0.0002>, rotation=<rot 0.0013 89.9994 89.9999>}]
 dyn_collar_01_r_MSTR [{shapeType=Capsule, length=0.01, radius=0.01}]
 dyn_lowerArmband_01_r [{shapeType=Capsule, length=5.269, radius=2.3814, center=<2.494 0 0.0002>, rotation=<rot 0.0013 89.9994 89.9999>}]
 dyn_lowerArmband_01_r_MSTR [{shapeType=Capsule, length=0.01, radius=0.01}]
 dyn_shirt_01_r [{shapeType=Box, extentX=12.637, extentY=3.8321, extentZ=13.0328, center=<6.3255 0 0>, rotation=<rot 0 0 89.9997>}]
 dyn_shirt_01_r_MSTR [{shapeType=Capsule, length=0.01, radius=0.01}]
 dyn_shirt_bk_01 [{shapeType=Capsule, length=4.7971, radius=4.1033, center=<3.2085 0 0>, rotation=<rot 0 90 90>}]
 dyn_shirt_bk_01_MSTR [{shapeType=Capsule, length=0.01, radius=0.01}]
 dyn_shirt_fr_01_r [{shapeType=Box, extentX=13.8283, extentY=4.7306, extentZ=8.0625, center=<4.1013 1.9891 1.9316>, rotation=<rot 0 0 89.9994>}]
 dyn_shirt_fr_01_r_MSTR [{shapeType=Capsule, length=0.01, radius=0.01}]
 dyn_shoelace_01_l [{shapeType=Capsule, length=5.269, radius=2.3814, center=<2.494 0 0.0002>, rotation=<rot 0.0013 89.9994 89.9999>}]
 dyn_shoelace_01_l_MSTR [{shapeType=Capsule, length=0.01, radius=0.01}]
 dyn_shoelace_01_r [{shapeType=Capsule, length=5.269, radius=2.3814, center=<2.494 0 0.0002>, rotation=<rot 0.0013 89.9994 89.9999>}]
 dyn_shoelace_01_r_MSTR [{shapeType=Capsule, length=0.01, radius=0.01}]
 dyn_shoelace_02_l [{shapeType=Capsule, length=5.269, radius=2.3814, center=<2.494 0 0.0002>, rotation=<rot 0.0013 89.9994 89.9999>}]
 dyn_shoelace_02_l_MSTR [{shapeType=Capsule, length=0.01, radius=0.01}]
 dyn_shoelace_02_r [{shapeType=Capsule, length=5.269, radius=2.3814, center=<2.494 0 0.0002>, rotation=<rot 0.0013 89.9994 89.9999>}]
 dyn_shoelace_02_r_MSTR [{shapeType=Capsule, length=0.01, radius=0.01}]
 dyn_tie_01_MSTR [{shapeType=Capsule, length=0.01, radius=0.01}]
 dyn_upperArmband_01_r [{shapeType=Capsule, length=5.269, radius=2.3814, center=<2.494 0 0.0002>, rotation=<rot 0.0013 89.9994 89.9999>}]
 dyn_upperArmband_01_r_MSTR [{shapeType=Capsule, length=0.01, radius=0.01}]
 lowerarm_l [{shapeType=Capsule, length=17.5, radius=4.75, center=<11.8488 0 0>, rotation=<rot -90 0.0264 -0.0263>}]
 lowerarm_r [{shapeType=Capsule, length=17.5, radius=4.75, center=<-11.8488 0 0>, rotation=<rot -90 -179.9802 -0.0197>}]
 pelvis [{shapeType=Capsule, length=11.5, radius=10.75}]
 spine_02 [{shapeType=Capsule, length=11.5, radius=10.75}]
 spine_04 [{shapeType=Capsule, length=11.5, radius=10.75}]
 spine_05 [{shapeType=Capsule, length=12.5, radius=9.75}]
 thigh_r [{shapeType=Capsule, length=33.5, radius=6.75, center=<19.8348 -1.9336 0.0001>, rotation=<rot -90 0.0049 -0.0049>}]
 upperarm_l [{shapeType=Capsule, length=15.5, radius=6.75, center=<11.8488 0 0>, rotation=<rot -90 0.0264 -0.0263>}]
 upperarm_r [{shapeType=Capsule, length=15.5, radius=6.75, center=<-11.8488 0 0>, rotation=<rot -90 -179.9802 -0.0197>}]
}
constraints [
 {bone1Name=spine_02, bone2Name=pelvis, swing1LimitDegrees=45, swing2LimitDegrees=45, twistLimitDegrees=45}
 {bone1Name=spine_04, bone2Name=spine_02, swing1LimitDegrees=45, swing2LimitDegrees=45, twistLimitDegrees=45}
 {bone1Name=spine_05, bone2Name=spine_04, swing1LimitDegrees=45, swing2LimitDegrees=45, twistLimitDegrees=45}
 {
  bone1Name=dyn_shoelace_01_l_MSTR
  bone2Name=pelvis
  swing1LimitDegrees=45
  swing1Motion=Free
  swing2LimitDegrees=45
  swing2Motion=Free
  twistLimitDegrees=45
  twistMotion=Free
 }
 {
  bone1Name=dyn_shoelace_02_l_MSTR
  bone2Name=pelvis
  swing1LimitDegrees=45
  swing1Motion=Free
  swing2LimitDegrees=45
  swing2Motion=Free
  twistLimitDegrees=45
  twistMotion=Free
 }
 {
  bone1Name=dyn_shoelace_01_r_MSTR
  bone2Name=thigh_r
  swing1LimitDegrees=45
  swing1Motion=Free
  swing2LimitDegrees=45
  swing2Motion=Free
  twistLimitDegrees=45
  twistMotion=Free
 }
 {
  bone1Name=dyn_shoelace_02_r_MSTR
  bone2Name=thigh_r
  swing1LimitDegrees=45
  swing1Motion=Free
  swing2LimitDegrees=45
  swing2Motion=Free
  twistLimitDegrees=45
  twistMotion=Free
 }
 {
  bone1Name=dyn_tie_01_MSTR
  bone2Name=spine_05
  swing1LimitDegrees=45
  swing1Motion=Free
  swing2LimitDegrees=45
  swing2Motion=Free
  twistLimitDegrees=45
  twistMotion=Free
 }
 {
  bone1Name=upperarm_l
  bone2Name=clavicle_l
  swing1LimitDegrees=45
  swing1Motion=Free
  swing2LimitDegrees=45
  swing2Motion=Free
  twistLimitDegrees=45
  twistMotion=Free
 }
 {
  bone1Name=upperarm_r
  bone2Name=clavicle_r
  swing1LimitDegrees=45
  swing1Motion=Free
  swing2LimitDegrees=45
  swing2Motion=Free
  twistLimitDegrees=45
  twistMotion=Free
 }
 {
  bone1Name=dyn_collar_01_l_MSTR
  bone2Name=spine_05
  swing1LimitDegrees=45
  swing1Motion=Free
  swing2LimitDegrees=45
  swing2Motion=Free
  twistLimitDegrees=45
  twistMotion=Free
 }
 {
  bone1Name=dyn_collar_01_r_MSTR
  bone2Name=spine_05
  swing1LimitDegrees=45
  swing1Motion=Free
  swing2LimitDegrees=45
  swing2Motion=Free
  twistLimitDegrees=45
  twistMotion=Free
 }
 {
  bone1Name=dyn_shirt_bk_01_MSTR
  bone2Name=pelvis
  swing1LimitDegrees=45
  swing1Motion=Free
  swing2LimitDegrees=45
  swing2Motion=Free
  twistLimitDegrees=45
  twistMotion=Free
 }
 {
  bone1Name=dyn_beltbuckle_01_l_MSTR
  bone2Name=pelvis
  swing1LimitDegrees=45
  swing1Motion=Free
  swing2LimitDegrees=45
  swing2Motion=Free
  twistLimitDegrees=45
  twistMotion=Free
 }
 {
  bone1Name=dyn_shirt_fr_01_r_MSTR
  bone2Name=pelvis
  swing1LimitDegrees=45
  swing1Motion=Free
  swing2LimitDegrees=45
  swing2Motion=Free
  twistLimitDegrees=45
  twistMotion=Free
 }
 {bone1Name=dyn_collar_01_l, bone2Name=dyn_collar_01_l_MSTR, swing1LimitDegrees=20, swing2LimitDegrees=20, twistLimitDegrees=5}
 {bone1Name=dyn_collar_01_r, bone2Name=dyn_collar_01_r_MSTR, swing1LimitDegrees=20, swing2LimitDegrees=20, twistLimitDegrees=5}
 {bone1Name=dyn_shoelace_01_l, bone2Name=dyn_shoelace_01_l_MSTR, swing1LimitDegrees=45, swing2LimitDegrees=45, twistLimitDegrees=5}
 {bone1Name=dyn_shoelace_02_l, bone2Name=dyn_shoelace_02_l_MSTR, swing1LimitDegrees=45, swing2LimitDegrees=45, twistLimitDegrees=5}
 {bone1Name=dyn_shoelace_01_r, bone2Name=dyn_shoelace_01_r_MSTR, swing1LimitDegrees=45, swing2LimitDegrees=45, twistLimitDegrees=5}
 {bone1Name=dyn_shoelace_02_r, bone2Name=dyn_shoelace_02_r_MSTR, swing1LimitDegrees=45, swing2LimitDegrees=45, twistLimitDegrees=5}
 {bone1Name=dyn_shirt_bk_01, bone2Name=dyn_shirt_bk_01_MSTR, swing1LimitDegrees=20, swing2LimitDegrees=45, twistLimitDegrees=5}
 {bone1Name=dyn_beltbuckle_01_l, bone2Name=dyn_beltbuckle_01_l_MSTR, swing1LimitDegrees=45, swing2LimitDegrees=45, twistLimitDegrees=5}
 {bone1Name=dyn_shirt_fr_01_r, bone2Name=dyn_shirt_fr_01_r_MSTR, swing1LimitDegrees=15, swing2LimitDegrees=25, twistLimitDegrees=5}
 {
  bone1Name=lowerarm_l
  bone2Name=upperarm_l
  swing1LimitDegrees=45
  swing1Motion=Free
  swing2LimitDegrees=45
  swing2Motion=Free
  twistLimitDegrees=45
  twistMotion=Free
 }
 {
  bone1Name=lowerarm_r
  bone2Name=upperarm_r
  swing1LimitDegrees=45
  swing1Motion=Free
  swing2LimitDegrees=45
  swing2Motion=Free
  twistLimitDegrees=45
  twistMotion=Free
 }
 {
  bone1Name=thigh_r
  bone2Name=pelvis
  swing1LimitDegrees=45
  swing1Motion=Free
  swing2LimitDegrees=45
  swing2Motion=Free
  twistLimitDegrees=45
  twistMotion=Free
 }
 {
  bone1Name=dyn_shirt_01_r_MSTR
  bone2Name=pelvis
  swing1LimitDegrees=45
  swing1Motion=Free
  swing2LimitDegrees=45
  swing2Motion=Free
  twistLimitDegrees=45
  twistMotion=Free
 }
 {bone1Name=dyn_shirt_01_r, bone2Name=dyn_shirt_01_r_MSTR, swing1LimitDegrees=20, swing2LimitDegrees=45, twistLimitDegrees=5}
 {
  bone1Name=dyn_upperArmband_01_r_MSTR
  bone2Name=upperarm_r
  swing1LimitDegrees=45
  swing1Motion=Free
  swing2LimitDegrees=45
  swing2Motion=Free
  twistLimitDegrees=45
  twistMotion=Free
 }
 {
  bone1Name=dyn_lowerArmband_01_r_MSTR
  bone2Name=upperarm_r
  swing1LimitDegrees=45
  swing1Motion=Free
  swing2LimitDegrees=45
  swing2Motion=Free
  twistLimitDegrees=45
  twistMotion=Free
 }
 {bone1Name=dyn_upperArmband_01_r, bone2Name=dyn_upperArmband_01_r_MSTR, swing1LimitDegrees=45, swing2LimitDegrees=45, twistLimitDegrees=5}
 {bone1Name=dyn_lowerArmband_01_r, bone2Name=dyn_lowerArmband_01_r_MSTR, swing1LimitDegrees=45, swing2LimitDegrees=45, twistLimitDegrees=5}
 {
  bone1Name=dyn_collarOuter_bk_01_l_MSTR
  bone2Name=spine_05
  swing1LimitDegrees=45
  swing1Motion=Free
  swing2LimitDegrees=45
  swing2Motion=Free
  twistLimitDegrees=45
  twistMotion=Free
 }
 {
  bone1Name=dyn_collarOuter_bk_01_r_MSTR
  bone2Name=spine_05
  swing1LimitDegrees=45
  swing1Motion=Free
  swing2LimitDegrees=45
  swing2Motion=Free
  twistLimitDegrees=45
  twistMotion=Free
 }
 {bone1Name=dyn_collarOuter_bk_01_l, bone2Name=dyn_collarOuter_bk_01_l_MSTR, swing1LimitDegrees=5, swing2LimitDegrees=40, twistLimitDegrees=5}
 {bone1Name=dyn_collarOuter_bk_01_r, bone2Name=dyn_collarOuter_bk_01_r_MSTR, swing1LimitDegrees=5, swing2LimitDegrees=40, twistLimitDegrees=5}
 {bone1Name=dyn_collarOuter_01_r, bone2Name=dyn_collarOuter_01_r_MSTR, swing1LimitDegrees=40, swing2LimitDegrees=40, twistLimitDegrees=5}
 {
  bone1Name=dyn_collarOuter_01_l_MSTR
  bone2Name=spine_05
  swing1LimitDegrees=45
  swing1Motion=Free
  swing2LimitDegrees=45
  swing2Motion=Free
  twistLimitDegrees=45
  twistMotion=Free
 }
 {bone1Name=dyn_collarOuter_01_l, bone2Name=dyn_collarOuter_01_l_MSTR, swing1LimitDegrees=40, swing2LimitDegrees=40, twistLimitDegrees=5}
 {
  bone1Name=dyn_collarOuter_01_r_MSTR
  bone2Name=spine_05
  swing1LimitDegrees=45
  swing1Motion=Free
  swing2LimitDegrees=45
  swing2Motion=Free
  twistLimitDegrees=45
  twistMotion=Free
 }
 {
  bone1Name=clavicle_l
  bone2Name=spine_05
  swing1LimitDegrees=45
  swing1Motion=Free
  swing2LimitDegrees=45
  swing2Motion=Free
  twistLimitDegrees=45
  twistMotion=Free
 }
 {
  bone1Name=clavicle_r
  bone2Name=spine_05
  swing1LimitDegrees=45
  swing1Motion=Free
  swing2LimitDegrees=45
  swing2Motion=Free
  twistLimitDegrees=45
  twistMotion=Free
 }
]
